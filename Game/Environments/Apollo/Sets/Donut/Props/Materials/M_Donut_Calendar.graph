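# Default Lit material for a calendar prop in the Donut set that draws day numbers, crossed-out days and one circled target day.
# UV1 becomes a seven-column cell index, offset by weekday and wrapped for adjacent-month days, which are tinted separately.
# Numbers come from a digit atlas sampled with explicit derivatives, colored by weekday.
# X and circle masks share a per-day pseudo-random rotation and multiply over the base texture.
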Material /Game/Environments/Apollo/Sets/Donut/Props/Materials/M_Donut_Calendar
graph {
 properties {pixelDepthOffsetMode=PDOM_Legacy, usedShadingModels=Default Lit}
 nodes [
  n1 = ScalarParameter() {defaultValue=0.143, parameterName=uv_x}
  n2 = ScalarParameter() {defaultValue=0.1871, parameterName=uv_y}
  n3 = Add(
   A=Divide(A=TextureCoordinate() {coordinateIndex=1}, B=AppendVector(A=@n1, B=@n2))
   B=AppendVector(A=ScalarParameter() {defaultValue=0.0356, parameterName=uv_offset_x}, B=ScalarParameter() {defaultValue=-0.315, parameterName=uv_offset_y})
  )
  n4 = ComponentMask(None=@n3) {g=true}
  n5 = Add(A=Ceil(None=ComponentMask(None=@n3) {r=true}), B=Multiply(A=Floor(None=@n4), B=Constant() {r=7}))
  n6 = Add(A=@n5, B=ScalarParameter() {defaultValue=-3, parameterName=OffsetWeekDay})
  n7 = ScalarParameter() {defaultValue=30, parameterName=NumberofDaysinMonth}
  n8 = Constant()
  n9 = If(A=@n6, B=@n7, "A > B"=Constant() {r=1}, "A == B"=@n8, "A < B"=@n8)
  n10 = Constant() {r=31}
  n11 = Multiply(A=Frac(None=Divide(A=@n7) {constB=10})) {constA=10, constB=10}
  n12 = Constant() {r=1}
  n13 = If(A=@n6, B=Constant(), "A > B"=Constant(), "A == B"=@n12, "A < B"=@n12)
  n14 = LinearInterpolate(A=LinearInterpolate(A=@n6, B=Subtract(A=@n6, B=@n7), Alpha=@n9), B=Add(A=@n6, B=Subtract(A=@n10, B=@n11)), Alpha=@n13)
  Multiply(A=@n14)
  n15 = Divide(A=@n14, B=ScalarParameter() {defaultValue=6.0433, parameterName=randomrotate})
  n16 = Rotator(Coordinate=Frac(None=@n3), Time=Multiply(A=Frac(None=@n15), B=ScalarParameter() {defaultValue=2.0325, parameterName=rotate}))
  n17 = Ceil(None=ComponentMask(None=@n3) {g=true})
  n18 = Add(A=@n5, B=ScalarParameter() {parameterName=OffsetCrossout})
  n19 = ScalarParameter() {defaultValue=19, parameterName=NumberoffCrossout}
  n20 = Constant()
  n21 = Constant() {r=1}
  n22 = If(A=@n18, B=Constant(), "A > B"=@n21, "A == B"=@n21, "A < B"=@n20)
  n23 = LinearInterpolate(
   A=Constant() {r=1}
   B=TextureSample(UVs=@n16) {texture=/Game/Environments/Apollo/Sets/Donut/Props/Textures/T_Xmarkspot_M}.R
   Alpha=Multiply(A=@n17, B=If(A=@n18, B=@n19, "A > B"=@n20, "A == B"=@n22, "A < B"=@n22))
  )
  n24 = MaterialFunctionCall(UVs=@n16, Texture Scale=ScalarParameter() {defaultValue=1.1112, parameterName=ScaleTarget}) {materialFunction=&ScaleUVsByCenter}
  n25 = Constant()
  n26 = If(
   A=Subtract(A=@n5, B=ScalarParameter() {defaultValue=33, parameterName=OffsetTarget})
   B=Constant() {r=1}
   "A > B"=@n25
   "A == B"=Constant() {r=1}
   "A < B"=@n25
  )
  n27 = LinearInterpolate(
   A=VectorParameter() {parameterName=ColorTarget, defaultValue=<rgba 1 0.0265 0 1>}.RGB
   B=Constant3Vector() {constant=<rgba 1 1 1 0>}
   Alpha=LinearInterpolate(
    A=Constant() {r=1}
    B=TextureSample(UVs=@n24.UVs) {texture=/Game/Environments/Apollo/Sets/Donut/Props/Textures/T_CircleMark_M}.R
    Alpha=Multiply(A=@n17, B=@n26)
   )
  )
  n28 = Constant() {r=7}
  n29 = Divide(A=Add(A=@n5) {constB=-1}, B=@n28)
  n30 = Frac(None=@n29)
  n31 = Constant()
  n32 = LinearInterpolate(
   A=VectorParameter() {parameterName=NumbersColor, defaultValue=<rgba 1 0 0 1>}.RGB
   B=VectorParameter() {parameterName=NumbersColor2, defaultValue=<rgba 0 0.0431 1 1>}.RGB
   Alpha=If(A=Multiply(A=@n30, B=@n28), B=Constant() {r=5}, "A > B"=@n31, "A == B"=@n31, "A < B"=Constant() {r=1})
  )
  n33 = VectorParameter() {parameterName=NumbersColor3, defaultValue=<rgba 0.2396 0.2396 0.2396 1>}
  n34 = Add(A=@n9, B=@n13)
  n35 = ScalarParameter() {defaultValue=-0.1348, group=Numbers, parameterName=u_NumberPaddingofseet}
  n36 = ScalarParameter() {defaultValue=0.3048, group=Numbers, parameterName=v_NumberPaddingoffset}
  n37 = Multiply(A=Add(A=Frac(None=@n3), B=AppendVector(A=@n35, B=@n36)))
  n38 = ScalarParameter() {defaultValue=0.2843, group=Numbers, parameterName=u_NumberPaddingScale}
  n39 = ScalarParameter() {defaultValue=0.3048, group=Numbers, parameterName=v_NumberPaddingScale}
  n40 = ScalarParameter() {defaultValue=0.4652, group=Numbers, parameterName=u_NumberPaddingScale2}
  n41 = ScalarParameter() {defaultValue=0.3143, group=Numbers, parameterName=v_NumberPaddingScale2}
  n42 = Constant() {r=1}
  n43 = If(A=@n14, B=Constant() {desc=Number of digits, r=10}, "A > B"=@n42, "A == B"=@n42, "A < B"=Constant())
  n44 = MaterialFunctionCall(UVs=@n37, Texture Scale=LinearInterpolate(A=AppendVector(A=@n38, B=@n39), B=AppendVector(A=@n40, B=@n41), Alpha=@n43)) {materialFunction=&ScaleUVsByCenter}
  n45 = LinearInterpolate(A=Constant() {r=1}, B=Constant() {r=2}, Alpha=@n43)
  n46 = Multiply(A=ComponentMask(None=@n44.UVs) {r=true}, B=@n45) {constB=5}
  n47 = Constant() {r=1}
  n48 = Constant() {r=4}
  n49 = Constant() {r=7}
  n50 = ComponentMask(None=@n46) {r=true}
  n51 = Constant() {r=10}
  n52 = Divide(A=@n14, B=@n51)
  n53 = Multiply(A=Frac(None=@n52), B=@n51)
  n54 = LinearInterpolate(A=@n53, B=Floor(None=@n52), Alpha=@n43)
  n55 = If(A=Constant() {r=1}, B=@n50, "A > B"=@n54, "A < B"=@n53)
  n56 = If(A=Constant() {r=2}, B=@n50, "A > B"=@n55, "A < B"=ScalarParameter() {defaultValue=-1, group=Numbers, parameterName=Digit_3})
  n57 = If(A=Constant() {r=3}, B=@n50, "A > B"=@n56, "A < B"=ScalarParameter() {defaultValue=4, group=Numbers, parameterName=Digit_4})
  n58 = If(A=Constant() {r=4}, B=@n50, "A > B"=@n57, "A < B"=ScalarParameter() {defaultValue=5, group=Numbers, parameterName=Digit_5})
  n59 = If(A=Constant() {r=5}, B=@n50, "A > B"=@n58, "A < B"=ScalarParameter() {defaultValue=-1, group=Numbers, parameterName=Digit_6})
  n60 = If(A=Constant() {r=6}, B=@n50, "A > B"=@n59, "A < B"=ScalarParameter() {defaultValue=7, group=Numbers, parameterName=Digit_7})
  n61 = If(A=Constant() {r=7}, B=@n50, "A > B"=@n60, "A < B"=ScalarParameter() {defaultValue=8, group=Numbers, parameterName=Digit_8})
  n62 = If(A=@n45, B=Constant() {r=8}, "A > B"=@n61, "A == B"=@n61, "A < B"=@n61)
  n63 = If(
   A=@n45
   B=Constant() {r=5}
   "A > B"=If(A=@n45, B=Constant() {r=6}, "A > B"=If(A=@n45, B=@n49, "A > B"=@n62, "A == B"=@n60, "A < B"=@n60), "A == B"=@n59, "A < B"=@n59)
   "A == B"=@n58
   "A < B"=@n58
  )
  n64 = If(
   A=@n45
   B=Constant() {r=2}
   "A > B"=If(A=@n45, B=Constant() {r=3}, "A > B"=If(A=@n45, B=@n48, "A > B"=@n63, "A == B"=@n57, "A < B"=@n57), "A == B"=@n56, "A < B"=@n56)
   "A == B"=@n55
   "A < B"=@n55
  )
  n65 = Add(A=Frac(None=@n46), B=Floor(None=If(A=@n45, B=@n47, "A > B"=@n64, "A == B"=@n54, "A < B"=@n54)))
  n66 = ScalarParameter() {defaultValue=11, group=Numbers, parameterName=Numbers of Digits On Texture}
  n67 = TextureSampleParameter2D!(
   UVs=AppendVector(A=Divide(A=@n65, B=@n66) {constB=11}, B=ComponentMask(None=@n44.UVs) {g=true})
   "DDX(UVs)"=DDX(Value=@n37)
   "DDY(UVs)"=DDY(Value=@n37)
   Apply View MipBias
  ) {
   group=Numbers
   mipValueMode=TMVM_Derivative
   parameterName=NumbersTexture
   samplerType=SAMPLERTYPE_Masks
   texture=/Game/Athena/Playlists/Barrier/Textures/T_Numbers2_M
  }
  n68 = Multiply(A=@n67.R, B=@n44.0-1 mask)
  Multiply(
   A=Multiply(A=TextureSample() {texture=/Game/Environments/Apollo/Sets/Donut/Props/Textures/T_Donut_Calendar_01_D}.RGB, B=Multiply(A=@n23, B=@n27))
   B=LinearInterpolate(A=Constant3Vector() {constant=<rgba 1 1 1 0>}, B=LinearInterpolate(A=@n32, B=@n33.RGB, Alpha=@n34), Alpha=@n68)
  )
  OneMinus(None=@n68)
  Reroute(None=@n44.0-1 mask)
  Saturate(None=@n29)
  ScalarParameter() {defaultValue=2, group=Numbers, parameterName=Digit_2}
  ScalarParameter() {defaultValue=1, group=Numbers, parameterName=Digit_1}
  TextureCoordinate() {coordinateIndex=2}
  TextureSample() {samplerType=SAMPLERTYPE_Normal, texture=/Game/Environments/Apollo/Sets/Donut/Props/Textures/T_Donut_Calendar_01_N}
  TextureSample() {samplerType=SAMPLERTYPE_Masks, texture=/Game/Environments/Apollo/Sets/Donut/Props/Textures/T_Donut_Calendar_01_S}
 ]
}
